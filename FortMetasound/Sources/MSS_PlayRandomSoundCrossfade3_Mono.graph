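# Mono one-shot MetaSound source that crossfades random close, medium and distant sounds by attenuation distance.
# Distance, mapped against MaxRange, samples three wavetable curves from the Curves bank to set each layer's gain.
# On play, each layer's gain is compared with zero to gate its MSS_PlayRandomSound_Mono, which receives shared pitch and volume ranges.
# The weighted layers are mixed, and OnFinished fires once all three have finished or been skipped.
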MetaSoundSource /FortMetasound/Sources/MSS_PlayRandomSoundCrossfade3_Mono
inputs [
 Curves {type=WaveTableBankAsset}
 MaxPitch {type=Float}
 MaxRange {type=Float}
 MaxVolume {type=Float}
 MinPitch {type=Float}
 MinVolume {type=Float}
 Sounds.Close {type=WaveAsset:Array}
 Sounds.Distant {type=WaveAsset:Array}
 Sounds.Medium {type=WaveAsset:Array}
 UE.Attenuation.Distance {type=Float}
 UE.Source.OnPlay {type=Trigger}
]
interfaces [UE.Attenuation:1.0, UE.OutputFormat.Mono:1.0, UE.Source.OneShot:1.0, UE.Source:1.1]
outputs [
 "UE.OutputFormat.Mono.Audio:0" {type=Audio}
 UE.Source.OneShot.OnFinished {type=Trigger}
]
pages [Default {
  input_defaults {
   Curves=null
   MaxPitch=1.05
   MaxRange=1024
   MaxVolume=1.05
   MinPitch=0.95
   MinVolume=0.95
   UE.Attenuation.Distance=0
   UE.Source.OnPlay=null
   Sounds.Close []
   Sounds.Distant []
   Sounds.Medium []
  }
  nodes [
   n1 = UE.WaveTableEvaluate(
    Input=MapRange.MapRange:Float(In Range B=^MaxRange, In=^UE.Attenuation.Distance).Out Value
    WaveTable=UE.MetasoundWaveTableGet(WaveTableBank=^Curves) {also {defaults {TableIndex=2}, version [1,1]}}.Out
   )
   n2 = TriggerCompare.Clamp:Float(Compare=^UE.Source.OnPlay, A=@n1.Output) {also {defaults {B=0, Type=3}}}
   n3 = &MSS_PlayRandomSound_Mono(MaxPitch=^MaxPitch, MaxVolume=^MaxVolume, MinPitch=^MinPitch, MinVolume=^MinVolume, Sounds=^Sounds.Distant, UE.Source.OnPlay=@n2.True)
   n4 = UE.WaveTableEvaluate(
    Input=MapRange.MapRange:Float(In Range B=^MaxRange, In=^UE.Attenuation.Distance).Out Value
    WaveTable=UE.MetasoundWaveTableGet(WaveTableBank=^Curves) {also {version [1,1]}}.Out
   )
   n5 = TriggerCompare.Clamp:Float(Compare=^UE.Source.OnPlay, A=@n4.Output) {also {defaults {B=0, Type=3}}}
   n6 = &MSS_PlayRandomSound_Mono(MaxPitch=^MaxPitch, MaxVolume=^MaxVolume, MinPitch=^MinPitch, MinVolume=^MinVolume, Sounds=^Sounds.Close, UE.Source.OnPlay=@n5.True)
   n7 = UE.WaveTableEvaluate(
    Input=MapRange.MapRange:Float(In Range B=^MaxRange, In=^UE.Attenuation.Distance).Out Value
    WaveTable=UE.MetasoundWaveTableGet(WaveTableBank=^Curves) {also {defaults {TableIndex=1}, version [1,1]}}.Out
   )
   n8 = TriggerCompare.Clamp:Float(Compare=^UE.Source.OnPlay, A=@n7.Output) {also {defaults {B=0, Type=3}}}
   n9 = &MSS_PlayRandomSound_Mono(MaxPitch=^MaxPitch, MaxVolume=^MaxVolume, MinPitch=^MinPitch, MinVolume=^MinVolume, Sounds=^Sounds.Medium, UE.Source.OnPlay=@n8.True)
   AudioMixerAudioMixerMono3_0 = "AudioMixer.Audio Mixer (Mono, 3)"(
    In 2=UE.Multiply:Audio by Float(PrimaryOperand=@n3."UE.OutputFormat.Mono.Audio:0", AdditionalOperands=@n1.Output).Out
    In 0=UE.Multiply:Audio by Float(PrimaryOperand=@n6."UE.OutputFormat.Mono.Audio:0", AdditionalOperands=@n4.Output).Out
    In 1=UE.Multiply:Audio by Float(PrimaryOperand=@n9."UE.OutputFormat.Mono.Audio:0", AdditionalOperands=@n7.Output).Out
   )
   TriggerAnyTriggerAny2_0 = "TriggerAny.Trigger Any (2)"(In 0=@n6.UE.Source.OneShot.OnFinished, In 1=@n5.False)
   TriggerAnyTriggerAny2_1 = "TriggerAny.Trigger Any (2)"(In 0=@n9.UE.Source.OneShot.OnFinished, In 1=@n8.False)
   TriggerAnyTriggerAny2_2 = "TriggerAny.Trigger Any (2)"(In 0=@n3.UE.Source.OneShot.OnFinished, In 1=@n2.False)
   TriggerAccumulatorTriggerAccumulate3_0 = "TriggerAccumulator.Trigger Accumulate (3)"(In 0=@TriggerAnyTriggerAny2_0.Out, In 1=@TriggerAnyTriggerAny2_1.Out, In 2=@TriggerAnyTriggerAny2_2.Out)
  ]
  links [
   {dst="@UE.OutputFormat.Mono.Audio:0", src=AudioMixerAudioMixerMono3_0.Out}
   {dst="@UE.Source.OneShot.OnFinished", src=TriggerAccumulatorTriggerAccumulate3_0.Out}
  ]
 }]
properties {assetImportData=&self:AssetImportData}
