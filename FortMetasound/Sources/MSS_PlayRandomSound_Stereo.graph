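# Reusable one-shot stereo MetaSound source that plays a randomly picked wave from the Sounds array with randomized pitch and volume.
# On play, a seeded random array get triggers the stereo wave player from Start Time.
# Random pitch and volume within min/max ranges multiply the base inputs, with pitch converted to semitones and volume scaling both channels.
# It also outputs the chosen wave's duration and fires OnFinished when playback ends.
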MetaSoundSource /FortMetasound/Sources/MSS_PlayRandomSound_Stereo
inputs [
 MaxPitch {type=Float}
 MaxVolume {type=Float}
 MinPitch {type=Float}
 MinVolume {type=Float}
 Pitch {type=Float}
 Seed {type=Int32}
 Shared State Behavior {type=Enum:SharedStateBehaviorType}
 Sounds {type=WaveAsset:Array}
 Start Time {type=Time}
 UE.Source.OnPlay {type=Trigger}
 Volume {type=Float}
]
interfaces [UE.OutputFormat.Stereo:1.0, UE.Source.OneShot:1.0, UE.Source:1.1]
outputs [
 "UE.OutputFormat.Stereo.Audio:0" {type=Audio}
 "UE.OutputFormat.Stereo.Audio:1" {type=Audio}
 UE.Source.OneShot.OnFinished {type=Trigger}
 WaveDuration {type=Time}
]
pages [Default {
  input_defaults {
   MaxPitch=1
   MaxVolume=1
   MinPitch=1
   MinVolume=1
   Pitch=1
   Seed=-1
   Shared State Behavior=1
   Start Time=0
   UE.Source.OnPlay=null
   Volume=1
   Sounds []
  }
  nodes [
   ArrayRandomGet_0 = Array.Random Get:WaveAsset:Array(Seed=^Seed, Shared State Behavior=^Shared State Behavior, In Array=^Sounds, Next=^UE.Source.OnPlay) {also {defaults {Enable Shared State="<MetaSoundBuilderResult.SUCCEEDED: 0>", No Repeats=-1, Weights []}}}
   GetWaveDuration_0 = "UE.Get Wave Duration: "(Wave=@ArrayRandomGet_0.Value)
   n1 = UE.Multiply:Float(AdditionalOperands=^Volume, PrimaryOperand=UE.RandomFloat(Max=^MaxVolume, Min=^MinVolume, Seed=^Seed, Next=^UE.Source.OnPlay).Value)
   FrequencyMultipliertoSemitone_0 = UE.Frequency Multiplier to Semitone:Float(Frequency Multiplier=UE.Multiply:Float(AdditionalOperands=^Pitch, PrimaryOperand=UE.RandomFloat(Max=^MaxPitch, Min=^MinPitch, Seed=^Seed, Next=^UE.Source.OnPlay).Value).Out)
   WavePlayer_0 = UE.Wave Player:Stereo(
    Start Time=^Start Time
    Play=@ArrayRandomGet_0.On Next
    Wave Asset=@ArrayRandomGet_0.Value
    Pitch Shift=@FrequencyMultipliertoSemitone_0.Semitones
   ) {also {version [1,1]}}
   Multiply_0 = UE.Multiply:Audio by Float(AdditionalOperands=@n1.Out, PrimaryOperand=@WavePlayer_0.Out Right)
   Multiply_1 = UE.Multiply:Audio by Float(AdditionalOperands=@n1.Out, PrimaryOperand=@WavePlayer_0.Out Left)
  ]
  links [
   {dst="@WaveDuration", src=GetWaveDuration_0.Duration}
   {dst="@UE.OutputFormat.Stereo.Audio:1", src=Multiply_0.Out}
   {dst="@UE.OutputFormat.Stereo.Audio:0", src=Multiply_1.Out}
   {dst="@UE.Source.OneShot.OnFinished", src=WavePlayer_0.On Finished}
  ]
 }]
properties {assetImportData=&self:AssetImportData, numChannels=2, outputFormat=Stereo, sampleRate=48000}
