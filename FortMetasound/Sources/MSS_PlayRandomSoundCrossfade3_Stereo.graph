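# Stereo one-shot MetaSound source that plays random sounds from Close, Medium and Distant sets, crossfaded by listener distance.
# Attenuation distance, mapped against MaxRange, reads three WaveTable bank curves whose values scale each layer before a gain-controlled stereo mixer.
# On play, each curve value gates its MSS_PlayRandomSound_Stereo layer, which shares pitch and volume ranges with the others.
# OnFinished fires once all three layers have either finished or been skipped.
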MetaSoundSource /FortMetasound/Sources/MSS_PlayRandomSoundCrossfade3_Stereo
inputs [
 Curves {type=WaveTableBankAsset}
 MaxPitch {type=Float}
 MaxRange {type=Float}
 MaxVolume {type=Float}
 MinPitch {type=Float}
 MinVolume {type=Float}
 Sounds.Close {type=WaveAsset:Array}
 Sounds.Distant {type=WaveAsset:Array}
 Sounds.Medium {type=WaveAsset:Array}
 UE.Attenuation.Distance {type=Float}
 UE.Source.OnPlay {type=Trigger}
 Volume.CloseGain {type=Float}
 Volume.DistantGain {type=Float}
 Volume.MediumGain {type=Float}
]
interfaces [UE.Attenuation:1.0, UE.OutputFormat.Stereo:1.0, UE.Source.OneShot:1.0, UE.Source:1.1]
outputs [
 "UE.OutputFormat.Stereo.Audio:0" {type=Audio}
 "UE.OutputFormat.Stereo.Audio:1" {type=Audio}
 UE.Source.OneShot.OnFinished {type=Trigger}
]
pages [Default {
  input_defaults {
   Curves=null
   MaxPitch=1.05
   MaxRange=1024
   MaxVolume=1.05
   MinPitch=0.95
   MinVolume=0.95
   UE.Attenuation.Distance=0
   UE.Source.OnPlay=null
   Volume.CloseGain=1
   Volume.DistantGain=1
   Volume.MediumGain=1
   Sounds.Close []
   Sounds.Distant []
   Sounds.Medium []
  }
  nodes [
   n1 = InitVariable.Audio()
   n2 = UE.WaveTableEvaluate(
    Input=MapRange.MapRange:Float(In Range B=^MaxRange, In=^UE.Attenuation.Distance).Out Value
    WaveTable=UE.MetasoundWaveTableGet(WaveTableBank=^Curves) {also {version [1,1]}}.Out
   )
   n3 = TriggerCompare.Clamp:Float(Compare=^UE.Source.OnPlay, A=@n2.Output) {also {defaults {Type=3}}}
   n4 = &MSS_PlayRandomSound_Stereo(MaxPitch=^MaxPitch, MaxVolume=^MaxVolume, MinPitch=^MinPitch, MinVolume=^MinVolume, Sounds=^Sounds.Close, UE.Source.OnPlay=@n3.True)
   n5 = UE.Multiply:Audio by Float(PrimaryOperand=@n4."UE.OutputFormat.Stereo.Audio:0", AdditionalOperands=@n2.Output)
   n6 = InitVariable.Audio()
   n7 = UE.WaveTableEvaluate(
    Input=MapRange.MapRange:Float(In Range B=^MaxRange, In=^UE.Attenuation.Distance).Out Value
    WaveTable=UE.MetasoundWaveTableGet(WaveTableBank=^Curves) {also {defaults {TableIndex=1}, version [1,1]}}.Out
   )
   n8 = TriggerCompare.Clamp:Float(Compare=^UE.Source.OnPlay, A=@n7.Output) {also {defaults {B=0, Type=3}}}
   n9 = &MSS_PlayRandomSound_Stereo(MaxPitch=^MaxPitch, MaxVolume=^MaxVolume, MinPitch=^MinPitch, MinVolume=^MinVolume, Sounds=^Sounds.Medium, UE.Source.OnPlay=@n8.True)
   n10 = UE.Multiply:Audio by Float(PrimaryOperand=@n9."UE.OutputFormat.Stereo.Audio:1", AdditionalOperands=@n7.Output)
   n11 = InitVariable.Audio()
   n12 = UE.Multiply:Audio by Float(PrimaryOperand=@n4."UE.OutputFormat.Stereo.Audio:1", AdditionalOperands=@n2.Output)
   n13 = InitVariable.Audio()
   n14 = UE.WaveTableEvaluate(
    Input=MapRange.MapRange:Float(In Range B=^MaxRange, In=^UE.Attenuation.Distance).Out Value
    WaveTable=UE.MetasoundWaveTableGet(WaveTableBank=^Curves) {also {defaults {TableIndex=2}, version [1,1]}}.Out
   )
   n15 = TriggerCompare.Clamp:Float(Compare=^UE.Source.OnPlay, A=@n14.Output) {also {defaults {B=0, Type=3}}}
   n16 = &MSS_PlayRandomSound_Stereo(MaxPitch=^MaxPitch, MaxVolume=^MaxVolume, MinPitch=^MinPitch, MinVolume=^MinVolume, Sounds=^Sounds.Distant, UE.Source.OnPlay=@n15.True)
   n17 = UE.Multiply:Audio by Float(PrimaryOperand=@n16."UE.OutputFormat.Stereo.Audio:0", AdditionalOperands=@n14.Output)
   n18 = InitVariable.Audio()
   n19 = UE.Multiply:Audio by Float(PrimaryOperand=@n16."UE.OutputFormat.Stereo.Audio:1", AdditionalOperands=@n14.Output)
   n20 = InitVariable.Audio()
   n21 = UE.Multiply:Audio by Float(PrimaryOperand=@n9."UE.OutputFormat.Stereo.Audio:0", AdditionalOperands=@n7.Output)
   AudioMixerAudioMixerStereo3_0 = "AudioMixer.Audio Mixer (Stereo, 3)"(
    Gain 0=^Volume.CloseGain
    Gain 2=^Volume.DistantGain
    Gain 1=^Volume.MediumGain
    In 0 L=VariableAccessor.Audio(Variable=VariableMutator.Audio(Variable=@n1.Variable, Value=@n5.Out).Variable).Value
    In 1 R=VariableAccessor.Audio(Variable=VariableMutator.Audio(Variable=@n6.Variable, Value=@n10.Out).Variable).Value
    In 0 R=VariableAccessor.Audio(Variable=VariableMutator.Audio(Variable=@n11.Variable, Value=@n12.Out).Variable).Value
    In 2 L=VariableAccessor.Audio(Variable=VariableMutator.Audio(Variable=@n13.Variable, Value=@n17.Out).Variable).Value
    In 2 R=VariableAccessor.Audio(Variable=VariableMutator.Audio(Variable=@n18.Variable, Value=@n19.Out).Variable).Value
    In 1 L=VariableAccessor.Audio(Variable=VariableMutator.Audio(Variable=@n20.Variable, Value=@n21.Out).Variable).Value
   )
   n22 = InitVariable.Trigger()
   TriggerAnyTriggerAny2_0 = "TriggerAny.Trigger Any (2)"(In 0=@n9.UE.Source.OneShot.OnFinished, In 1=@n8.False)
   n23 = InitVariable.Trigger()
   TriggerAnyTriggerAny2_2 = "TriggerAny.Trigger Any (2)"(In 0=@n4.UE.Source.OneShot.OnFinished, In 1=@n3.False)
   n24 = InitVariable.Trigger()
   TriggerAnyTriggerAny2_1 = "TriggerAny.Trigger Any (2)"(In 0=@n16.UE.Source.OneShot.OnFinished, In 1=@n15.False)
   TriggerAccumulatorTriggerAccumulate3_0 = "TriggerAccumulator.Trigger Accumulate (3)"(
    In 1=VariableAccessor.Trigger(Variable=VariableMutator.Trigger(Variable=@n22.Variable, Value=@TriggerAnyTriggerAny2_0.Out).Variable).Value
    In 0=VariableAccessor.Trigger(Variable=VariableMutator.Trigger(Variable=@n23.Variable, Value=@TriggerAnyTriggerAny2_2.Out).Variable).Value
    In 2=VariableAccessor.Trigger(Variable=VariableMutator.Trigger(Variable=@n24.Variable, Value=@TriggerAnyTriggerAny2_1.Out).Variable).Value
   )
  ]
  links [
   {dst="@UE.OutputFormat.Stereo.Audio:0", src=AudioMixerAudioMixerStereo3_0.Out L}
   {dst="@UE.OutputFormat.Stereo.Audio:1", src=AudioMixerAudioMixerStereo3_0.Out R}
   {dst="@UE.Source.OneShot.OnFinished", src=TriggerAccumulatorTriggerAccumulate3_0.Out}
  ]
 }]
properties {assetImportData=&self:AssetImportData, numChannels=2, outputFormat=Stereo}
